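# WBP_Sidebar_TextChat_MessageList is the sidebar variant of the text chat message list widget, inheriting all its layout from WBP_TextChat_MessageList.
# GetDesiredEntryClassForItem picks a channel button entry for conversation view models, a custom button entry for custom button view models, and otherwise defers to the parent.
# Item clicks broadcast OnCustomListButtonClicked with the ButtonType for custom button entries, or OnItemClicked with the item for anything else.
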WidgetBlueprint /Sidebar/Panels/Chat/List/WBP_Sidebar_TextChat_MessageList
graphs [
 EventGraph {nodes [
   n1 = "On Item Clicked (ListView)"!() -> (OutputDelegate, then, Item:Object Reference) {is=title K2Node_ComponentBoundEvent}
   > n2 = Cast To BP_Sidebar_CustomButtonEntryVM!(execute, Object=@n1.Item) -> (then, CastFailed, AsBP Sidebar Custom Button Entry VM:BP Sidebar Custom Button Entry VM Object Reference) {is=title K2Node_DynamicCast}
   > Call On Custom List Button Clicked!(
    execute:Exec
    self:WBP Sidebar Text Chat Message List Object Reference
    ButtonType:Name=Get ButtonType!(self:BP Sidebar Custom Button Entry VM Object Reference=@n2) -> (ButtonType:Name=None)
   ) -> (then:Exec) {is=title K2Node_CallDelegate}
   Call On Item Clicked!(execute:Exec=@n2.CastFailed, self:WBP Sidebar Text Chat Message List Object Reference, Item:Object Reference=@n1.Item) -> (then:Exec) {is=title K2Node_CallDelegate}
  ]}
 GetDesiredEntryClassForItem {nodes [
   n1 = GetDesiredEntryClassForItem!() -> (then, Item:Object Reference) {is=title K2Node_FunctionEntry}
   > n2 = Cast To FortTextChatConversationVM!(execute, Object=@n1) -> (then, CastFailed, AsFort Text Chat Conversation VM:Fort Text Chat Conversation VM Object Reference) {is=title K2Node_DynamicCast}
   > Return Node!(execute, ReturnValue:User Widget Class Reference=/Sidebar/Panels/Chat/WBP_Sidebar_ChannelButton.WBP_Sidebar_ChannelButton_C)
   n3 = Cast To BP_Sidebar_CustomButtonEntryVM(execute=@n2.CastFailed, Object=@n1)
   > Return Node(ReturnValue=/Sidebar/Panels/Social/List/WBP_Sidebar_CustomButton_ListEntry.WBP_Sidebar_CustomButton_ListEntry_C)
   n4 = Parent: GetDesiredEntryClassForItem!(execute=@n3.CastFailed, Item:Object Reference=@n1) -> (then, ReturnValue:User Widget Class Reference) {is=title K2Node_CallParentFunction}
   > Return Node(ReturnValue=@n4)
  ]}
 OnCustomListButtonClicked {nodes [OnCustomListButtonClicked!() -> (then, ButtonType:Name) {is=title K2Node_FunctionEntry}]}
 OnItemClicked {nodes [OnItemClicked!() -> (then, Item:Object Reference) {is=title K2Node_FunctionEntry}]}
]
variables [
 OnCustomListButtonClicked {}
 OnItemClicked {}
]
widgets {
 info {
  inheritedWidgetCount=0
  parentClass=/FortniteCoreUI/UI/Social/Chat/TextChat/WBP_TextChat_MessageList.WBP_TextChat_MessageList_C
  rootWidgetClass=null
  widgetCount=0
 }
 widgets []
}
