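# Party avatar widget for the sidebar chat panel, a single Image drawn with the MI_Sidebar_PartyButtonAvatar material.
# PreConstruct caches the image's dynamic material instance.
# InitWithSocialUser passes a Fort social user to the player avatar viewmodel.
# SetIsBehind and SetIsPending convert bools to scalar material parameters, so the material handles both visual states.
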WidgetBlueprint /Sidebar/Panels/Chat/WBP_Sidebar_PartyAvatar
graphs [
 EventGraph {nodes [
   Event PreConstruct()
   > n1 = GetDynamicMaterial(self=Get AvatarImage() -> (AvatarImage:Image Object Reference))
   > Set Dynamic Material(Dynamic Material:Material Instance Dynamic Object Reference=@n1) -> (Output_Get:Material Instance Dynamic Object Reference)
  ]}
 InitWithSocialUser {nodes [
   n1 = InitWithSocialUser!() -> (then, InSocialUser:Fort Social User Object Reference) {is=title K2Node_FunctionEntry}
   > SetSocialUser!(
    execute
    self:Fort Social User Viewmodel Object Reference=Get FortPlayerAvatarViewModel() -> (FortPlayerAvatarViewModel:Fort Player Avatar Viewmodel Object Reference)
    InSocialUser:Fort Social User Object Reference=@n1
   ) -> (then)
  ]}
 SetIsBehind {nodes [
   n1 = SetIsBehind!() -> (then, Is Avatar Behind:bool=~) {is=title K2Node_FunctionEntry}
   > SetScalarParameterValue(
    self=Get Dynamic Material() -> (Dynamic Material:Material Instance Dynamic Object Reference)
    ParameterName=IsBehind
    Value="To Float (Boolean)"(InBool=@n1)
   )
  ]}
 SetIsPending {nodes [
   n1 = SetIsPending!() -> (then, Is Pending:bool=~) {is=title K2Node_FunctionEntry}
   > SetScalarParameterValue(self=Get Dynamic Material(), ParameterName=isPending, Value="To Float (Boolean)"(InBool=@n1))
  ]}
]
variables [Dynamic Material {type=MaterialInstanceDynamic}]
widgets {
 info {inheritedWidgetCount=0, parentClass=&UserWidget, rootWidgetClass=&Image, widgetCount=1}
 widgets [AvatarImage {
   widgetClassPath=&Image
   bIsVariable=true
   widget=AvatarImage
   properties {brush {
     bIsDynamicallyLoaded=false
     drawAs=Image
     imageType=FullColor
     mirroring=NoMirror
     resourceName=None
     resourceObject=/Sidebar/Materials/MI_Sidebar_PartyButtonAvatar
     tiling=NoTile
     imageSize=<42 42>
     margin=<ltrb 0 0 0 0>
     outlineSettings {
      bUseBrushTransparency=false
      roundingType=HalfHeightRadius
      width=0
      color {colorUseRule=UseColor_Specified, specifiedColor=<rgba 0 0 0 0>}
      cornerRadii=<0 0 0 0>
     }
     tintColor {colorUseRule=UseColor_Specified, specifiedColor=<rgba 1 1 1 1>}
     uVRegion {bIsValid=false, max=<0 0>, min=<0 0>}
    }}
  }]
}
