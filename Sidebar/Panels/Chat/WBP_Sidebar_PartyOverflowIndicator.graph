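# WBP_Sidebar_PartyOverflowIndicator is a CommonUserWidget circular badge in the sidebar chat panel, showing a party overflow count as "+N".
# An Overlay stacks a circle-material Image behind centered text, which SetPartyOverflowCount fills through a localized format.
# SetIsPartyJoined stores the flag and recolors both to green when joined or gray otherwise, with PreConstruct applying the not-joined look.
# SetPendingStyle separately applies a dark pending palette without changing the flag.
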WidgetBlueprint /Sidebar/Panels/Chat/WBP_Sidebar_PartyOverflowIndicator
graphs [
 EventGraph {nodes [
   Event PreConstruct()
   > SetIsPartyJoined!(execute, self, IsPartyJoined:bool=false) -> (then)
   Event Construct()
   Event Tick()
  ]}
 SetIsPartyJoined {nodes [
   n1 = SetIsPartyJoined!() -> (then, IsPartyJoined:bool=~) {is=title K2Node_FunctionEntry}
   > n2 = Set IsPartyJoined(IsPartyJoined:bool=@n1) -> (Output_Get:bool)
   > n3 = Branch(Condition=@n2)
   > SetColorAndOpacity(
    self:Image Object Reference=Get Image_BG() -> (Image_BG:Image Object Reference)
    InColorAndOpacity:Linear Color Structure=(R=0.013,G=0.0529,B=0.0152,A=1)
   )
   > SetColorAndOpacity(
    self:Text Object Reference=Get Text_OverflowCount() -> (Text_OverflowCount:Common Text Object Reference)
    InColorAndOpacity:Slate Color Structure=Make Slate Color(SpecifiedColor=(R=0.2831,G=0.807,B=0.2918,A=1))
   )
   SetColorAndOpacity(execute=@n3.else, self:Image Object Reference=Get Image_BG(), InColorAndOpacity:Linear Color Structure=(R=0.4427,G=0.4427,B=0.4427,A=1))
   > SetColorAndOpacity(
    self:Text Object Reference=Get Text_OverflowCount()
    InColorAndOpacity:Slate Color Structure=Make Slate Color(SpecifiedColor=(R=0.0573,G=0.0573,B=0.0573,A=1))
   )
  ]}
 SetPartyOverflowCount {nodes [
   n1 = SetPartyOverflowCount!() -> (then, OverflowCount:int) {is=title K2Node_FunctionEntry}
   > "SetText (Text)"(self=Get Text_OverflowCount(), InText=Format Text!(Format="LOCTEXT(\"+{count}\")", count:int=@n1) -> (Result))
  ]}
 SetPendingStyle {nodes [
   SetPendingStyle() {is=title K2Node_FunctionEntry}
   > SetColorAndOpacity(self:Image Object Reference=Get Image_BG(), InColorAndOpacity:Linear Color Structure=(R=0.0222,G=0.0212,B=0.0284,A=1))
   > SetColorAndOpacity(
    self:Text Object Reference=Get Text_OverflowCount()
    InColorAndOpacity:Slate Color Structure=Make Slate Color(SpecifiedColor=(R=0.5776,G=0.5776,B=0.5776,A=1))
   )
  ]}
]
variables [IsPartyJoined {type=boolean}]
widgets {
 info {inheritedWidgetCount=0, parentClass=&CommonUserWidget, rootWidgetClass=&Overlay, widgetCount=3}
 widgets [
  Overlay_0 {widgetClassPath=&Overlay, widget=Overlay_0}
  Image_BG {
   widgetClassPath=&Image
   bIsVariable=true
   parent=Overlay_0
   widget=Image_BG
   properties {brush {
     bIsDynamicallyLoaded=false
     drawAs=Image
     imageType=FullColor
     mirroring=NoMirror
     resourceName=None
     resourceObject=/Game/Athena/UI/Materials/M_UI_Circle
     tiling=NoTile
     imageSize=<42 42>
     margin=<ltrb 0 0 0 0>
     outlineSettings {
      bUseBrushTransparency=false
      roundingType=HalfHeightRadius
      width=0
      color {colorUseRule=UseColor_Specified, specifiedColor=<rgba 0 0 0 0>}
      cornerRadii=<0 0 0 0>
     }
     tintColor {colorUseRule=UseColor_Specified, specifiedColor=<rgba 1 1 1 1>}
     uVRegion {bIsValid=false, max=<0 0>, min=<0 0>}
    }}
   slot {class=OverlaySlot, properties {horizontalAlignment=HAlign_Fill, verticalAlignment=VAlign_Fill}}
  }
  Text_OverflowCount {
   widgetClassPath=&CommonTextBlock
   bIsVariable=true
   parent=Overlay_0
   widget=Text_OverflowCount
   properties {
    text=+X
    colorAndOpacity {colorUseRule=UseColor_Specified, specifiedColor=<rgba 0.2831 0.807 0.2918 1>}
    font {
     bForceMonospaced=false
     bMaterialIsStencil=false
     fontMaterial=null
     fontObject=&HeadingNow
     letterSpacing=0
     monospacedWidth=1
     size=15
     skewAmount=0
     typefaceFontName=74 Regular
     outlineSettings {
      bApplyOutlineToDropShadows=false
      bMiteredCorners=false
      bSeparateFillAlpha=false
      outlineMaterial=null
      outlineSize=0
      outlineColor=<rgba 0 0 0 1>
     }
    }
    shadowOffset=<0 0>
   }
   slot {class=OverlaySlot, properties {horizontalAlignment=HAlign_Center, verticalAlignment=VAlign_Center}}
  }
 ]
}
