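# Sidebar chat widget showing a conversation party's members or pending invitees as a row of overlapping avatars.
# AddAvatarsAndHandleOverlap and AddPendingAvatars rebuild a negatively spaced DynamicEntryBox, marking all but the list's last avatar as behind and appending an overflow indicator past MaximumAvatarsAllowed.
# The member path breaks one index later than the pending path and broadcasts isAvatarListNotEmpty from size and PartyState.
# SetIsPartyJoined forwards joined state to the overflow indicator.
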WidgetBlueprint /Sidebar/Panels/Chat/WBP_Sidebar_PartyAvatarList
graphs [
 AddAvatarsAndHandleOverlap {nodes [
   n1 = AddAvatarsAndHandleOverlap!() -> (then, Conversation Party Members:Array of Fort Social User Viewmodel Object References) {is=title K2Node_FunctionEntry}
   > n2 = Set Party Size(Party Size:int=Length(TargetArray:Array of Fort Social User Viewmodel Object References=@n1)) -> (Output_Get:int=0)
   > Reset!(
    execute
    self:Dynamic Entry Box Object Reference=Get DynamicEntryBox_PartyAvatars() -> (DynamicEntryBox_PartyAvatars:Dynamic Entry Box Object Reference)
    bDeleteWidgets:bool=true
   ) -> (then)
   n3 = Get PartyState() -> (PartyState:EFortConversationPartyState Enum=None)
   n4 = "Equal (Enum)"(A:EFortConversationPartyState Enum=@n3, B:EFortConversationPartyState Enum=Joined)
   n5 = "Equal (Enum)"(A=@n3, B=Active)
   n6 = "Equal (Enum)"(A=@n3, B=Joining)
   > Set with Broadcast isAvatarListNotEmpty!(execute, isAvatarListNotEmpty:bool=AND Boolean(A="integer > integer"(A=@n2), B=OR Boolean!(A=@n5, B=@n6, C:bool=@n4) -> (ReturnValue))) -> (then, Output_Get:bool)
   > Branch(Condition=Is Not Empty!(TargetArray:Array of Fort Social User Viewmodel Object References=@n1) -> (ReturnValue:bool))
   > n7 = For Each Loop with Break!(
    Exec:Exec
    Array:Array of Fort Social User Viewmodel Object References=Get Conversation Party Members() -> (Conversation Party Members:Array of Fort Social User Viewmodel Object References)
    Break:Exec=@n10.then
   ) -> (LoopBody:Exec, Array Element:Fort Social User Viewmodel Object Reference, Array Index:int, Completed:Exec)
   > n8 = Create Entry of Class!(
    execute
    self:Dynamic Entry Box Object Reference=Get DynamicEntryBox_PartyAvatars()
    EntryClass:User Widget Class Reference=&/WBP_Sidebar_PartyAvatar.WBP_Sidebar_PartyAvatar_C
   ) -> (then, ReturnValue:WBP Sidebar Party Avatar Object Reference)
   > InitWithSocialUser!(
    execute
    self:WBP Sidebar Party Avatar Object Reference=@n8
    InSocialUser:Fort Social User Object Reference=GetSocialUser!(self:Fort Social User Viewmodel Object Reference=@n7) -> (ReturnValue:Fort Social User Object Reference)
   ) -> (then)
   n9 = Get Party Size() -> (Party Size:int=0)
   > SetIsBehind!(
    execute
    self:WBP Sidebar Party Avatar Object Reference=@n8
    Is Avatar Behind:bool="Not Equal (Integer)"(A=@n7.Array Index, B=int - int(A=@n9))
   ) -> (then)
   > n10 = Branch(Condition="Equal (Integer)"(A=Get MaximumAvatarsAllowed() -> (MaximumAvatarsAllowed:int=0), B=@n7.Array Index))
   Branch(
    execute=@n7.Completed
    Condition="integer > integer"(A=Get Party Size() -> (Party Size=0), B=Get MaximumAvatarsAllowed() -> (MaximumAvatarsAllowed=0))
   )
   > n11 = Create Entry of Class(
    self=Get DynamicEntryBox_PartyAvatars()
    EntryClass=&/WBP_Sidebar_PartyOverflowIndicator.WBP_Sidebar_PartyOverflowIndicator_C
   ) -> (ReturnValue:WBP Sidebar Party Overflow Indicator Object Reference)
   > SetPartyOverflowCount!(
    execute
    self:WBP Sidebar Party Overflow Indicator Object Reference=@n11
    OverflowCount:int=int - int(A=Get Party Size() -> (Party Size=0), B=Get MaximumAvatarsAllowed() -> (MaximumAvatarsAllowed=0))
   ) -> (then)
   > SetIsPartyJoined!(execute, self:WBP Sidebar Party Overflow Indicator Object Reference=@n11, IsPartyJoined:bool=Get IsPartyJoined() -> (IsPartyJoined:bool)) -> (then)
  ]}
 AddPendingAvatars {nodes [
   n1 = AddPendingAvatars!() -> (then, Pending Party Members:Array of Fort Social User Object References) {is=title K2Node_FunctionEntry}
   > Reset(self=Get DynamicEntryBox_PartyAvatars(), bDeleteWidgets=true)
   > Set InvitedCount(InvitedCount:int=Length(TargetArray:Array of Fort Social User Object References=@n1)) -> (Output_Get:int=0)
   > Branch(Condition=Is Not Empty(TargetArray:Array of Fort Social User Object References=@n1))
   > n2 = For Each Loop with Break(
    Array:Array of Fort Social User Object References=Get Pending Party Members() -> (Pending Party Members:Array of Fort Social User Object References)
    Break=@n6.then
   ) -> (Array Element:Fort Social User Object Reference)
   > n3 = Create Entry of Class(self=Get DynamicEntryBox_PartyAvatars(), EntryClass=&WBP_Sidebar_PartyAvatar_C) -> (ReturnValue:WBP Sidebar Party Avatar Object Reference)
   > InitWithSocialUser(self=@n3, InSocialUser=@n2)
   n4 = Get InvitedCount() -> (InvitedCount:int=0)
   > SetIsBehind(self=@n3, Is Avatar Behind="Not Equal (Integer)"(A=@n2.Array Index, B=int - int(A=@n4)))
   > SetIsPending!(execute, self:WBP Sidebar Party Avatar Object Reference=@n3, Is Pending:bool=true) -> (then)
   n5 = Get MaximumAvatarsAllowed() -> (MaximumAvatarsAllowed=0)
   > n6 = Branch(Condition="Equal (Integer)"(A=int - int(A=@n5), B=@n2.Array Index))
   Branch(
    execute=@n2.Completed
    Condition="integer > integer"(A=Get InvitedCount() -> (InvitedCount=0), B=Get MaximumAvatarsAllowed() -> (MaximumAvatarsAllowed=0))
   )
   > n7 = Create Entry of Class(
    self=Get DynamicEntryBox_PartyAvatars()
    EntryClass=&WBP_Sidebar_PartyOverflowIndicator_C
   ) -> (ReturnValue:WBP Sidebar Party Overflow Indicator Object Reference)
   > SetPartyOverflowCount(self=@n7, OverflowCount=int - int(A=Get InvitedCount() -> (InvitedCount=0), B=Get MaximumAvatarsAllowed() -> (MaximumAvatarsAllowed=0)))
   > SetPendingStyle(self:WBP Sidebar Party Overflow Indicator Object Reference=@n7)
  ]}
 EventGraph {links [], nodes []}
 SetIsPartyJoined {nodes [
   n1 = SetIsPartyJoined!() -> (then, IsPartyJoined:bool=~) {is=title K2Node_FunctionEntry}
   > Set IsPartyJoined(IsPartyJoined:bool=@n1) -> (Output_Get:bool)
   > n2 = For Each Loop with Break(
    Array:Array of User Widget Object References=GetAllEntries!(self:Dynamic Entry Box Base Object Reference=Get DynamicEntryBox_PartyAvatars()) -> (ReturnValue:Array of User Widget Object References)
    Break=@n4.then
   ) -> (Array Element:User Widget Object Reference)
   > n3 = Cast To WBP_Sidebar_PartyOverflowIndicator!(execute, Object=@n2) -> (then, CastFailed, AsWBP Sidebar Party Overflow Indicator:WBP Sidebar Party Overflow Indicator Object Reference) {is=title K2Node_DynamicCast}
   > n4 = SetIsPartyJoined(self=@n3, IsPartyJoined=Get IsPartyJoined()) {is=title K2Node_CallFunction}
  ]}
]
variables [
 isAvatarListNotEmpty {type=boolean}
 IsPartyJoined {type=boolean}
 MaximumAvatarsAllowed {type=integer, default=4}
 PartyState {type=EFortConversationPartyState, default=None}
]
widgets {
 info {inheritedWidgetCount=0, parentClass=&UserWidget, rootWidgetClass=&Overlay, widgetCount=2}
 widgets [
  Overlay_0 {widgetClassPath=&Overlay, widget=Overlay_0, properties {bOverride_Cursor=true}}
  DynamicEntryBox_PartyAvatars {
   widgetClassPath=&DynamicEntryBox
   bIsVariable=true
   parent=Overlay_0
   widget=DynamicEntryBox_PartyAvatars
   properties {entryWidgetClass=&WBP_Sidebar_PartyAvatar_C, entrySpacing=<-10 0>}
   slot {class=OverlaySlot}
  }
 ]
}
