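# Sidebar chat button widget, derived from WBP_Sidebar_Button_Base, showing party conversation state with a headphones icon, member avatars and a label.
# SetPartyButtonVisuals broadcasts Party State, picks the label text and calls SetIsJoined, which updates the avatar list, icon tint and backplate color.
# Hover and unhover overrides call the parent, then reapply joined styling.
# Two helper functions return avatar and text visibility, hiding text when joined with avatars visible.
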WidgetBlueprint /Sidebar/Panels/Chat/WBP_Sidebar_PartyButton
graphs [
 EventGraph {nodes [
   Event On Hovered()
   > Parent: On Hovered() {is=title K2Node_CallParentFunction}
   > Branch(Condition=Get IsPartyJoined() -> (IsPartyJoined:bool))
   > SetBackplateColor!(
    execute
    self:WBP Sidebar Block Backplate Object Reference=Get Backplate!(self:WBP Sidebar Block Background Button Object Reference=Get Background() -> (Background:WBP Sidebar Block Background Button Object Reference)) -> (Backplate:WBP Sidebar Block Backplate Object Reference)
    Backplate Switch:E_UI_Sidebar_Button_Backplate_Color Enum=NewEnumerator3
   ) -> (then)
   Event On Unhovered()
   > Parent: On Unhovered() {is=title K2Node_CallParentFunction}
   > SetIsJoined!(execute, self, Is Party Joined:bool=Get IsPartyJoined()) -> (then)
  ]}
 SetIsJoined {nodes [
   n1 = SetIsJoined!() -> (then, Is Party Joined:bool=~) {is=title K2Node_FunctionEntry}
   > n2 = Set IsPartyJoined(IsPartyJoined:bool=@n1) -> (Output_Get:bool)
   > SetIsPartyJoined!(
    execute
    self:WBP Sidebar Party Avatar List Object Reference=Get PartyMemberAvatars() -> (PartyMemberAvatars:WBP Sidebar Party Avatar List Object Reference)
    IsPartyJoined:bool=@n2
   ) -> (then)
   > Set ColorAndOpacity!(
    execute
    ColorAndOpacity:Linear Color Structure=SelectColor(A=(R=0.0395,G=1,B=0.2307,A=1), B=(R=1,G=1,B=1,A=1), bPickA=Get IsPartyJoined())
    self:Image Object Reference=Get Icon_Voice_Status() -> (Icon_Voice_Status:Image Object Reference)
   ) -> (then, Output_Get:Linear Color Structure)
   > SetBackplateColor(
    self=Get Backplate(self=Get Background())
    Backplate Switch=Select(
     Option 0:E_UI_Sidebar_Button_Backplate_Color Enum=NewEnumerator1
     Option 1:E_UI_Sidebar_Button_Backplate_Color Enum=NewEnumerator3
     Index=Get IsPartyJoined()
    ) -> (ReturnValue:E_UI_Sidebar_Button_Backplate_Color Enum)
   )
  ]}
 SetPartyButtonAvatarListVisibility {nodes [
   n1 = SetPartyButtonAvatarListVisibility!() -> (then, isAvatarListVisible:bool=~) {is=title K2Node_FunctionEntry}
   > Return Node!(
    execute
    Visibility:ESlateVisibility Enum=Select(Option 0:ESlateVisibility Enum=Collapsed, Option 1:ESlateVisibility Enum=SelfHitTestInvisible, Index=@n1) -> (ReturnValue:ESlateVisibility Enum)
   )
  ]}
 SetPartyButtonTextVisibility {nodes [
   n1 = SetPartyButtonTextVisibility!() -> (then, isAvatarListVisible:bool=~, Party State:EFortConversationPartyState Enum=None) {is=title K2Node_FunctionEntry}
   n2 = "Equal (Enum)"(A:EFortConversationPartyState Enum=@n1.Party State, B:EFortConversationPartyState Enum=Joined)
   > Return Node(Visibility=Select(Option 0=SelfHitTestInvisible, Option 1=Collapsed, Index=AND Boolean(A=@n1, B=@n2)))
  ]}
 SetPartyButtonVisuals {nodes [
   n1 = SetPartyButtonVisuals!() -> (then, Party state:EFortConversationPartyState Enum=None) {is=title K2Node_FunctionEntry}
   > Set with Broadcast Party State!(execute, Party State:EFortConversationPartyState Enum=@n1) -> (then, Output_Get:EFortConversationPartyState Enum=None)
   > "SetText (Text)"(
    self=Get Button_Text() -> (Button_Text:Common Text Object Reference)
    InText=Select!(
     None:Text="LOCTEXT(\"Start party\")"
     Creating:Text="LOCTEXT(\"Joining\")"
     Active:Text="LOCTEXT(\"Join party\")"
     Joining:Text="LOCTEXT(\"Joining\")"
     Joined:Text="LOCTEXT(\"In party\")"
     Index:EFortConversationPartyState Enum=@n1
    ) -> (ReturnValue:Text)
   )
   > SetIsJoined(Is Party Joined="Equal (Enum)"(A=Get Party state() -> (Party state:EFortConversationPartyState Enum=None), B=Joined)) {is=title K2Node_CallFunction}
  ]}
]
variables [
 IsPartyJoined {type=boolean}
 Party State {type=EFortConversationPartyState, default=None}
]
widgets {
 info {
  inheritedWidgetCount=0
  namedSlotCount=1
  parentClass=/Sidebar/Buttons/WBP_Sidebar_Button_Base.WBP_Sidebar_Button_Base_C
  rootWidgetClass=null
  widgetCount=4
 }
 widgets [
  GridPanel_41 {widgetClassPath=&GridPanel, widget=GridPanel_41, properties {columnFill [0,1], rowFill [1]}}
  Icon_Voice_Status {
   widgetClassPath=&Image
   bIsVariable=true
   parent=GridPanel_41
   widget=Icon_Voice_Status
   properties {brush {
     bIsDynamicallyLoaded=false
     drawAs=Image
     imageType=FullColor
     mirroring=NoMirror
     resourceName=None
     resourceObject=/Sidebar/Materials/Icons/MI_Sidebar_Icon_Headphones
     tiling=NoTile
     imageSize=<32 32>
     margin=<ltrb 0 0 0 0>
     outlineSettings {
      bUseBrushTransparency=false
      roundingType=HalfHeightRadius
      width=0
      color {colorUseRule=UseColor_Specified, specifiedColor=<rgba 0 0 0 0>}
      cornerRadii=<0 0 0 0>
     }
     tintColor {colorUseRule=UseColor_Specified, specifiedColor=<rgba 1 1 1 1>}
     uVRegion {bIsValid=false, max=<0 0>, min=<0 0>}
    }}
   slot {class=GridSlot, properties {horizontalAlignment=HAlign_Left, verticalAlignment=VAlign_Center, padding=<ltrb 15 10 10 10>}}
  }
  PartyMemberAvatars {
   widgetClassPath=&/WBP_Sidebar_PartyAvatarList
   bIsVariable=true
   parent=GridPanel_41
   widget=PartyMemberAvatars
   properties {visibility=Collapsed}
   slot {class=GridSlot, properties {column=1, verticalAlignment=VAlign_Center, padding=<ltrb 0 0 12 0>}}
  }
  Button_Text {
   widgetClassPath=&CommonTextBlock
   bIsVariable=true
   parent=GridPanel_41
   widget=Button_Text
   properties {
    mobileFontSizeMultiplier=1.35
    text=Start party
    colorAndOpacity {colorUseRule=UseColor_Specified, specifiedColor=<rgba 1 1 1 0.65>}
    font {
     bForceMonospaced=false
     bMaterialIsStencil=false
     fontMaterial=null
     fontObject=&HeadingNow
     letterSpacing=0
     monospacedWidth=1
     size=15
     skewAmount=0
     typefaceFontName=74 Regular
     outlineSettings {
      bApplyOutlineToDropShadows=false
      bMiteredCorners=false
      bSeparateFillAlpha=false
      outlineMaterial=null
      outlineSize=0
      outlineColor=<rgba 0 0 0 1>
     }
    }
    shadowOffset=<0 0>
   }
   slot {class=GridSlot, properties {column=2, verticalAlignment=VAlign_Center, padding=<ltrb 0 0 12 0>}}
  }
 ]
}
