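# Sidebar widget that wraps the shared text chat restrictions widget, filling a vertical box and placing a sidebar-style action button in its named slot.
# PreConstruct passes exposed backplate visibility and padding to the inner widget.
# Construct reads the fort.UseEpicParties console variable to drive its updated visuals and SetPlasterBeach.
# Both button events request the sidebar view model to close the sidebar; the slotted button also triggers the restriction action handler.
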WidgetBlueprint /Sidebar/Panels/Chat/WBP_Sidebar_TextChat_Restrictions
graphs [EventGraph {nodes [
   "On Button Clicked (WBP_TextChat_Restrictions)"() {is=title K2Node_ComponentBoundEvent}
   n1 = Get FortSidebarVM() -> (FortSidebarVM:Fort Sidebar VM Object Reference)
   > RequestCloseSidebar(self:Fort Sidebar VM Object Reference=@n1)
   "On Clicked (Button_Restriction)"!() -> (OutputDelegate, then, Button:Common Button Base Object Reference) {is=title K2Node_ComponentBoundEvent}
   > RequestCloseSidebar(self=@n1)
   > HandleRestrictionActionButtonClicked(self:WBP Text Chat Restrictions Object Reference=Get WBP_TextChat_Restrictions() -> (WBP_TextChat_Restrictions:WBP Text Chat Restrictions Object Reference))
   Event Construct()
   > n2 = GetConsoleVariableBoolValue!(execute, VariableName:String=fort.UseEpicParties) -> (then, ReturnValue:bool)
   n3 = Get WBP_TextChat_Restrictions()
   > n4 = Set Use Updated Visuals!(execute, Use Updated Visuals:bool=@n2, self:WBP Text Chat Restrictions Object Reference=@n3) -> (then, Output_Get:bool)
   > SetPlasterBeach!(execute, self:WBP Text Chat Restrictions Object Reference=@n3, IsActive:bool=@n4) -> (then)
   Event PreConstruct()
   > SetBackplateVisible!(
    execute
    self:WBP Text Chat Restrictions Object Reference=Get WBP_TextChat_Restrictions()
    IsBackplateVisible:bool=Get ShowBackplate() -> (ShowBackplate:bool)
   ) -> (then)
   > SetPadding!(
    execute
    self:User Widget Object Reference=Get WBP_TextChat_Restrictions()
    InPadding:Margin Structure=Get RestrictionWidgetPadding() -> (RestrictionWidgetPadding:Margin Structure)
   ) -> (then)
  ]}]
variables [
 NotifyIsRestrictionWidgetVisible {type=boolean, default=true}
 RestrictionWidgetPadding {type=Margin, default=<ltrb 0 0 0 0>}
 ShowBackplate {type=boolean, default=true}
]
widgets {
 info {
  inheritedWidgetCount=0
  namedSlotCount=1
  parentClass=&CommonUserWidget
  rootWidgetClass=&VerticalBox
  widgetCount=3
 }
 widgets [
  VerticalBox_0 {widgetClassPath=&VerticalBox, widget=VerticalBox_0}
  WBP_TextChat_Restrictions {
   widgetClassPath=/FortniteCoreUI/UI/Social/Chat/TextChat/WBP_TextChat_Restrictions
   bIsVariable=true
   parent=VerticalBox_0
   widget=WBP_TextChat_Restrictions
   slot {class=VerticalBoxSlot, properties {size {sizeRule=Fill, value=1}}}
  }
  Button_Restriction {
   widgetClassPath=/Sidebar/Buttons/WBP_Sidebar_ButtonGeneric
   bIsVariable=true
   namedSlotHost=WBP_TextChat_Restrictions
   widget=Button_Restriction
   properties {minHeightOverride=54, text=Restriction Action}
  }
 ]
}
